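# Physics asset for the neck bone chain of the CanineCronutMix medium male player body cosmetic.
# Three spine capsules link through free constraints up to a small sphere on mstr_neck01, forming the base.
# From there a sphere and two boxes form the dyn_neck02–04 chain, with twist locked at each joint.
# Neck swing is limited, except the last joint, which also locks one swing axis to act as a hinge.
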PhysicsAsset /BRCosmetics/Characters/Player/Male/Medium/Bodies/M_MED_CanineCronutMix/Meshes/M_MED_CanineCronutMix_Physics_Neck
bodies {
 dyn_neck02 [{shapeType=Sphere, radius=1.3279, shapeName=dyn_neck02_sphere, center=<2.2059 0 0>}]
 dyn_neck03 [{
   shapeType=Box
   extentX=5.9887
   extentY=4.4283
   extentZ=2.3839
   shapeName=dyn_neck03_box
   center=<2.1919 0 0.1718>
   rotation=<rot -3.8178 0 0>
  }]
 dyn_neck04 [{
   shapeType=Box
   extentX=5.9098
   extentY=2.5407
   extentZ=10.4823
   shapeName=dyn_neck04_box
   center=<2.2774 0 0.4296>
   rotation=<rot 0 -89.4222 -88.9505>
  }]
 mstr_neck01 [{shapeType=Sphere, radius=1, shapeName=mstr_neck01_sphere}]
 spine_02 [{shapeType=Capsule, length=7.6332, radius=10.3662, shapeName=spine_05_capsule, center=<6.2058 -0.2808 -0.1726>, rotation=<rot 0 -90 0>}]
 spine_04 [{shapeType=Capsule, length=11.4609, radius=10.3662, shapeName=spine_05_capsule, center=<3.0956 0.8439 -0.1726>, rotation=<rot 0 -90 0>}]
 spine_05 [{shapeType=Capsule, length=11.4609, radius=10.3662, shapeName=spine_05_capsule, center=<-1.0984 -0.2808 -0.1726>, rotation=<rot 0 -90 0>}]
}
constraints [
 {
  bone1Name=spine_05
  bone2Name=spine_04
  swing1LimitDegrees=45
  swing1Motion=Free
  swing2LimitDegrees=45
  swing2Motion=Free
  twistLimitDegrees=45
  twistMotion=Free
 }
 {
  bone1Name=spine_04
  bone2Name=spine_02
  swing1LimitDegrees=45
  swing1Motion=Free
  swing2LimitDegrees=45
  swing2Motion=Free
  twistLimitDegrees=45
  twistMotion=Free
 }
 {
  bone1Name=mstr_neck01
  bone2Name=spine_05
  swing1LimitDegrees=45
  swing1Motion=Free
  swing2LimitDegrees=45
  swing2Motion=Free
  twistLimitDegrees=45
  twistMotion=Free
 }
 {
  bone1Name=dyn_neck04
  bone2Name=dyn_neck03
  swing1LimitDegrees=45
  swing1Motion=Locked
  swing2LimitDegrees=45
  twistLimitDegrees=45
  twistMotion=Locked
 }
 {bone1Name=dyn_neck02, bone2Name=mstr_neck01, swing1LimitDegrees=6, swing2LimitDegrees=25, twistLimitDegrees=45, twistMotion=Locked}
 {bone1Name=dyn_neck03, bone2Name=dyn_neck02, swing1LimitDegrees=25, swing2LimitDegrees=25, twistLimitDegrees=45, twistMotion=Locked}
]
