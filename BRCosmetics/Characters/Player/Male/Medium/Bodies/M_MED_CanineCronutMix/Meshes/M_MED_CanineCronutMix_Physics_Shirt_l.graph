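# Physics asset for the left shirt sleeve of the CanineCronutMix medium male player body, giving the cloth secondary motion around the upper arm.
# Front, back, top and bottom dyn bones carry capsules, each constrained to a small mstr anchor sphere with only swing2 unlocked, with tighter limits on bottom and back.
# Angularly free constraints also link the front-back and top-bottom pairs.
# An unconstrained upperarm_l capsule supplies the arm's shape.
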PhysicsAsset /BRCosmetics/Characters/Player/Male/Medium/Bodies/M_MED_CanineCronutMix/Meshes/M_MED_CanineCronutMix_Physics_Shirt_l
bodies {
 dyn_arm_shirt_back_l [{
   shapeType=Capsule
   length=5.674
   radius=1.5172
   shapeName=dyn_arm_shirt_bot_r_capsule
   center=<-9.7023 -0.0042 -0.7104>
   rotation=<rot -86.8268 -0.1178 0.142>
  }]
 dyn_arm_shirt_bot_l [{
   shapeType=Capsule
   length=5.674
   radius=1.5172
   shapeName=dyn_arm_shirt_bot_r_capsule
   center=<-7.4771 -0.0032 -0.587>
   rotation=<rot -86.8268 -0.1178 0.142>
  }]
 dyn_arm_shirt_front_l [{
   shapeType=Capsule
   length=5.674
   radius=1.5172
   shapeName=dyn_arm_shirt_bot_r_capsule
   center=<-9.6882 -0.0042 -1.0172>
   rotation=<rot -87.6907 -179.9673 179.9992>
  }]
 dyn_arm_shirt_top_l [{
   shapeType=Capsule
   length=5.389
   radius=1.8023
   shapeName=dyn_arm_shirt_bot_r_capsule
   center=<-13.131 -0.0056 -0.9004>
   rotation=<rot -86.8268 -0.1178 0.142>
  }]
 mstr_arm_shirt_back_l [{shapeType=Sphere, radius=0.505, shapeName=mstr_arm_shirt_back_l_sphere}]
 mstr_arm_shirt_bot_l [{shapeType=Sphere, radius=0.505, shapeName=mstr_arm_shirt_bot_l_sphere}]
 mstr_arm_shirt_front_l [{shapeType=Sphere, radius=0.505, shapeName=mstr_arm_shirt_front_l_sphere}]
 mstr_arm_shirt_top_l [{shapeType=Sphere, radius=0.505, shapeName=mstr_arm_shirt_top_l_sphere}]
 upperarm_l [{
   shapeType=Capsule
   length=8.8416
   radius=4.785
   shapeName=upperarm_r_capsule
   center=<13.9825 0 -0.881>
   rotation=<rot 86.2054 -179.9994 -179.9994>
  }]
}
constraints [
 {
  bone1Name=dyn_arm_shirt_top_l
  bone2Name=mstr_arm_shirt_top_l
  swing1LimitDegrees=45
  swing1Motion=Locked
  swing2LimitDegrees=45
  twistLimitDegrees=45
  twistMotion=Locked
 }
 {
  bone1Name=dyn_arm_shirt_front_l
  bone2Name=mstr_arm_shirt_front_l
  swing1LimitDegrees=45
  swing1Motion=Locked
  swing2LimitDegrees=45
  twistLimitDegrees=45
  twistMotion=Locked
 }
 {
  bone1Name=dyn_arm_shirt_bot_l
  bone2Name=mstr_arm_shirt_bot_l
  swing1LimitDegrees=45
  swing1Motion=Locked
  swing2LimitDegrees=15
  twistLimitDegrees=45
  twistMotion=Locked
 }
 {
  bone1Name=dyn_arm_shirt_back_l
  bone2Name=mstr_arm_shirt_back_l
  swing1LimitDegrees=45
  swing1Motion=Locked
  swing2LimitDegrees=7
  twistLimitDegrees=45
  twistMotion=Locked
 }
 {
  bone1Name=dyn_arm_shirt_back_l
  bone2Name=dyn_arm_shirt_front_l
  swing1LimitDegrees=45
  swing1Motion=Free
  swing2LimitDegrees=45
  swing2Motion=Free
  twistLimitDegrees=45
  twistMotion=Free
 }
 {
  bone1Name=dyn_arm_shirt_top_l
  bone2Name=dyn_arm_shirt_bot_l
  swing1LimitDegrees=45
  swing1Motion=Free
  swing2LimitDegrees=45
  swing2Motion=Free
  twistLimitDegrees=45
  twistMotion=Free
 }
]
