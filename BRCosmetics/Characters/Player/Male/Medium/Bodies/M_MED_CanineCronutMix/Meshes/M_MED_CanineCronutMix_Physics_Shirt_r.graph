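# Physics asset for the right shirt sleeve on the CanineCronutMix medium male player body, for secondary sleeve motion around the upper arm.
# Four dyn sleeve capsules (top, bottom, front, back) are each constrained to a small mstr anchor sphere, with twist and one swing axis locked, leaving hinge-like movement.
# Angularly free constraints also link opposite flaps, front to back and top to bottom.
# An upperarm_r capsule provides the arm's body volume.
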PhysicsAsset /BRCosmetics/Characters/Player/Male/Medium/Bodies/M_MED_CanineCronutMix/Meshes/M_MED_CanineCronutMix_Physics_Shirt_r
bodies {
 dyn_arm_shirt_back_r [{
   shapeType=Capsule
   length=5.4439
   radius=1.7474
   shapeName=dyn_arm_shirt_bot_r_capsule
   center=<9.7023 0.0042 0.7104>
   rotation=<rot -86.8268 -0.1178 0.142>
  }]
 dyn_arm_shirt_bot_r [{
   shapeType=Capsule
   length=5.674
   radius=1.5172
   shapeName=dyn_arm_shirt_bot_r_capsule
   center=<7.4771 0.0032 0.587>
   rotation=<rot -86.8268 -0.1178 0.142>
  }]
 dyn_arm_shirt_front_r [{
   shapeType=Capsule
   length=5.674
   radius=1.5172
   shapeName=dyn_arm_shirt_bot_r_capsule
   center=<9.6757 0.0042 0.7089>
   rotation=<rot -86.8268 -0.1178 0.142>
  }]
 dyn_arm_shirt_top_r [{
   shapeType=Capsule
   length=5.674
   radius=1.5172
   shapeName=dyn_arm_shirt_bot_r_capsule
   center=<13.131 0.0056 0.9004>
   rotation=<rot -86.8268 -0.1178 0.142>
  }]
 mstr_arm_shirt_back_r [{shapeType=Sphere, radius=0.505, shapeName=mstr_arm_shirt_back_r_sphere}]
 mstr_arm_shirt_bot_r [{shapeType=Sphere, radius=0.505, shapeName=mstr_arm_shirt_bot_r_sphere}]
 mstr_arm_shirt_front_r [{shapeType=Sphere, radius=0.505, shapeName=mstr_arm_shirt_front_r_sphere}]
 mstr_arm_shirt_top_r [{shapeType=Sphere, radius=0.505, shapeName=mstr_arm_shirt_top_r_sphere}]
 upperarm_r [{
   shapeType=Capsule
   length=9.262
   radius=4.3646
   shapeName=upperarm_r_capsule
   center=<-12.5582 0 0.9755>
   rotation=<rot 86.2054 -179.9994 -179.9994>
  }]
}
constraints [
 {
  bone1Name=dyn_arm_shirt_top_r
  bone2Name=mstr_arm_shirt_top_r
  swing1LimitDegrees=45
  swing1Motion=Locked
  swing2LimitDegrees=45
  twistLimitDegrees=45
  twistMotion=Locked
 }
 {
  bone1Name=dyn_arm_shirt_back_r
  bone2Name=mstr_arm_shirt_back_r
  swing1LimitDegrees=45
  swing1Motion=Locked
  swing2LimitDegrees=10
  twistLimitDegrees=45
  twistMotion=Locked
 }
 {
  bone1Name=dyn_arm_shirt_bot_r
  bone2Name=mstr_arm_shirt_bot_r
  swing1LimitDegrees=45
  swing1Motion=Locked
  swing2LimitDegrees=15
  twistLimitDegrees=45
  twistMotion=Locked
 }
 {
  bone1Name=dyn_arm_shirt_front_r
  bone2Name=mstr_arm_shirt_front_r
  swing1LimitDegrees=45
  swing1Motion=Locked
  swing2LimitDegrees=45
  twistLimitDegrees=45
  twistMotion=Locked
 }
 {
  bone1Name=dyn_arm_shirt_back_r
  bone2Name=dyn_arm_shirt_front_r
  swing1LimitDegrees=45
  swing1Motion=Free
  swing2LimitDegrees=45
  swing2Motion=Free
  twistLimitDegrees=45
  twistMotion=Free
 }
 {
  bone1Name=dyn_arm_shirt_top_r
  bone2Name=dyn_arm_shirt_bot_r
  swing1LimitDegrees=45
  swing1Motion=Free
  swing2LimitDegrees=45
  swing2Motion=Free
  twistLimitDegrees=45
  twistMotion=Free
 }
]
