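# Physics asset for the left leg of the shorts on the M_MED_CanineCronutMix male player cosmetic body, apparently for secondary cloth motion.
# Four dynamic-bone capsules (front, back, in, out) are each constrained to a small master-bone sphere, with twist and swing1 locked for hinge-like swing.
# Two extra constraints with free angular motion link the inner flap to the back and outer flaps.
# An unconstrained thigh capsule provides the leg's collision volume.
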PhysicsAsset /BRCosmetics/Characters/Player/Male/Medium/Bodies/M_MED_CanineCronutMix/Meshes/M_MED_CanineCronutMix_Physics_Shorts_L
bodies {
 dyn_short_back_l02 [{
   shapeType=Capsule
   length=8.695
   radius=1.3122
   shapeName=dyn_short_in_r02_capsule
   center=<-9.4488 -0.0013 -0.0457>
   rotation=<rot -88.4559 0.0108 0.0005>
  }]
 dyn_short_front_l02 [{
   shapeType=Capsule
   length=8.695
   radius=1.3122
   shapeName=dyn_short_in_r02_capsule
   center=<-9.0959 -0.0012 -0.0083>
   rotation=<rot -87.7002 -179.9785 179.9893>
  }]
 dyn_short_in_l02 [{
   shapeType=Capsule
   length=8.695
   radius=1.3122
   shapeName=dyn_short_in_r02_capsule
   center=<-7.3275 -0.0009 0.0298>
   rotation=<rot -86.9775 -179.9813 179.9921>
  }]
 dyn_short_out_l02 [{
   shapeType=Capsule
   length=8.695
   radius=1.3122
   shapeName=dyn_short_in_r02_capsule
   center=<-9.4927 -0.0013 -0.5941>
   rotation=<rot -86.9775 -179.9813 179.9921>
  }]
 mstr_short_back_l [{shapeType=Sphere, radius=0.505, shapeName=mstr_short_back_l_sphere}]
 mstr_short_front_l [{shapeType=Sphere, radius=0.505, shapeName=mstr_short_front_l_sphere}]
 mstr_short_in_l [{shapeType=Sphere, radius=0.505, shapeName=mstr_short_in_l_sphere}]
 mstr_short_out_l [{shapeType=Sphere, radius=0.505, shapeName=mstr_short_out_l_sphere}]
 thigh_l [{shapeType=Capsule, length=10.733, radius=6.1555, shapeName=thigh_l_capsule, center=<-25.6551 -1.8155 0.3257>, rotation=<rot 90 180 0>}]
}
constraints [
 {
  bone1Name=dyn_short_back_l02
  bone2Name=mstr_short_back_l
  swing1LimitDegrees=45
  swing1Motion=Locked
  swing2LimitDegrees=45
  twistLimitDegrees=45
  twistMotion=Locked
 }
 {
  bone1Name=dyn_short_front_l02
  bone2Name=mstr_short_front_l
  swing1LimitDegrees=45
  swing1Motion=Locked
  swing2LimitDegrees=45
  twistLimitDegrees=45
  twistMotion=Locked
 }
 {
  bone1Name=dyn_short_out_l02
  bone2Name=mstr_short_out_l
  swing1LimitDegrees=45
  swing1Motion=Locked
  swing2LimitDegrees=45
  twistLimitDegrees=45
  twistMotion=Locked
 }
 {
  bone1Name=dyn_short_in_l02
  bone2Name=mstr_short_in_l
  swing1LimitDegrees=45
  swing1Motion=Locked
  swing2LimitDegrees=45
  twistLimitDegrees=45
  twistMotion=Locked
 }
 {
  bone1Name=dyn_short_in_l02
  bone2Name=dyn_short_back_l02
  swing1LimitDegrees=45
  swing1Motion=Free
  swing2LimitDegrees=45
  swing2Motion=Free
  twistLimitDegrees=45
  twistMotion=Free
 }
 {
  bone1Name=dyn_short_in_l02
  bone2Name=dyn_short_out_l02
  swing1LimitDegrees=45
  swing1Motion=Free
  swing2LimitDegrees=45
  swing2Motion=Free
  twistLimitDegrees=45
  twistMotion=Free
 }
]
