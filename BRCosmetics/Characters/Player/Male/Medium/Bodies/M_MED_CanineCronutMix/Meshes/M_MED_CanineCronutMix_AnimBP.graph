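# Animation Blueprint for the CanineCronutMix male body cosmetic, adding secondary motion to its neck, shorts and shirt.
# The copied parent pose feeds a Control Rig, neck RigidBody, skydive-weighted additive shorts bone rotations, then per-side shorts and shirt RigidBody simulations.
# Each update fades those side simulations by calf and upper-arm rotation, damping shirts during skydiving or parachuting.
# That airborne state also drives WPO cloth wind, turbulence and drag on dynamic main and overlay materials.
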AnimBlueprint /BRCosmetics/Characters/Player/Male/Medium/Bodies/M_MED_CanineCronutMix/Meshes/M_MED_CanineCronutMix_AnimBP
target_skeleton=&/M_MED_CanineCronutMix_Skeleton
graphs [
 AnimGraph {nodes [
   RigidBody() {
    bClampLinearTranslationLimitToRefPose=true
    overridePhysicsAsset=&/M_MED_CanineCronutMix_Physics_Shirt_l
    simulationSpace=BaseBoneSpace
    baseBoneRef {boneName=upperarm_l}
    overrideWorldGravity {z=-980}
    simSpaceSettings {maxAngularAcceleration=10, maxAngularVelocity=4, maxLinearAcceleration=10, maxLinearVelocity=5, worldAlpha=1}
   }
   RigidBody() {
    bClampLinearTranslationLimitToRefPose=true
    overridePhysicsAsset=&/M_MED_CanineCronutMix_Physics_Shirt_r
    simulationSpace=BaseBoneSpace
    baseBoneRef {boneName=upperarm_r}
    overrideWorldGravity {z=-980}
    simSpaceSettings {maxAngularAcceleration=10, maxAngularVelocity=4, maxLinearAcceleration=10, maxLinearVelocity=5, worldAlpha=1}
   }
   n1 = Control Rig!(Source=Copy Pose From Mesh() {bUseAttachedParent=true}, Alpha, RB:bool=Get bRigidBodyEnabled(), FrontEnd:bool=Get bIsInFrontEnd()) -> (Pose) {controlRigAssetReference {blueprintRigClass=&/M_MED_CanineCronutMix_CtrlRig.M_MED_CanineCronutMix_CtrlRig_C}}
   n2 = RigidBody(ComponentPose=Local To Component(LocalPose=@n1)) {
    bClampLinearTranslationLimitToRefPose=true
    bOverrideWorldGravity=true
    overridePhysicsAsset=&/M_MED_CanineCronutMix_Physics_Neck
    simulationSpace=BaseBoneSpace
    baseBoneRef {boneName=spine_05}
    overrideWorldGravity {z=-980}
    simSpaceSettings {
     maxAngularAcceleration=40
     maxAngularVelocity=20
     maxLinearAcceleration=150
     maxLinearVelocity=20
     worldAlpha=0.2
     externalLinearDragV=<1 1 1>
    }
   }
   n3 = Get SkyDiving_Alpha() -> (SkyDiving_Alpha:double)
   n4 = "Transform (Modify) Bone - Bone: mstr_short_front_l"(ComponentPose=@n2, Alpha=@n3, Translation="0,0,0", Rotation="20,0,0", Scale="1,1,1") {rotationMode=BMM_Additive, rotationSpace=BCS_BoneSpace, boneToModify {boneName=mstr_short_front_l}, rotation {pitch=20}}
   n5 = "Transform (Modify) Bone - Bone: mstr_short_front_r"(ComponentPose=@n4, Alpha=@n3, Translation="0,0,0", Rotation="20,0,0", Scale="1,1,1") {rotationMode=BMM_Additive, rotationSpace=BCS_BoneSpace, boneToModify {boneName=mstr_short_front_r}, rotation {pitch=20}}
   n6 = "Transform (Modify) Bone - Bone: mstr_short_in_l"(ComponentPose=@n5, Alpha=@n3, Translation="0,0,0", Rotation="20,0,0", Scale="1,1,1") {rotationMode=BMM_Additive, rotationSpace=BCS_BoneSpace, boneToModify {boneName=mstr_short_in_l}, rotation {pitch=20}}
   n7 = "Transform (Modify) Bone - Bone: mstr_short_in_r"(ComponentPose=@n6, Alpha=@n3, Translation="0,0,0", Rotation="20,0,0", Scale="1,1,1") {rotationMode=BMM_Additive, rotationSpace=BCS_BoneSpace, boneToModify {boneName=mstr_short_in_r}, rotation {pitch=20}}
   n8 = "Transform (Modify) Bone - Bone: mstr_short_in_l"(ComponentPose=@n7, Alpha=@n3, Translation="0,0,0", Rotation="20,0,0", Scale="1,1,1") {rotationMode=BMM_Additive, rotationSpace=BCS_BoneSpace, boneToModify {boneName=mstr_short_in_l}, rotation {pitch=20}}
   n9 = "Transform (Modify) Bone - Bone: mstr_short_in_r"(ComponentPose=@n8, Alpha=@n3, Translation="0,0,0", Rotation="20,0,0", Scale="1,1,1") {rotationMode=BMM_Additive, rotationSpace=BCS_BoneSpace, boneToModify {boneName=mstr_short_in_r}, rotation {pitch=20}}
   n10 = RigidBody(ComponentPose=@n9, Alpha=Get Short_r_Alpha() -> (Short_r_Alpha:double)) {
    bClampLinearTranslationLimitToRefPose=true
    overridePhysicsAsset=&/M_MED_CanineCronutMix_Physics_Shorts_R
    simulationSpace=BaseBoneSpace
    baseBoneRef {boneName=thigh_twist_01_r}
    simSpaceSettings {maxAngularAcceleration=8, maxAngularVelocity=4, maxLinearAcceleration=8, maxLinearVelocity=4, worldAlpha=0.5}
   }
   n11 = RigidBody(ComponentPose=@n10, Alpha=Get Short_l_Alpha() -> (Short_l_Alpha:float)) {
    bClampLinearTranslationLimitToRefPose=true
    overridePhysicsAsset=&/M_MED_CanineCronutMix_Physics_Shorts_L
    simulationSpace=BaseBoneSpace
    baseBoneRef {boneName=thigh_l}
    simSpaceSettings {maxAngularAcceleration=8, maxAngularVelocity=4, maxLinearAcceleration=8, maxLinearVelocity=4, worldAlpha=0.5}
   }
   n12 = RigidBody(ComponentPose=@n11, Alpha=Get shirt_r_alpha() -> (shirt_r_alpha:double)) {
    bClampLinearTranslationLimitToRefPose=true
    overridePhysicsAsset=&M_MED_CanineCronutMix_Physics_Shirt_r
    simulationSpace=BaseBoneSpace
    baseBoneRef {boneName=upperarm_r}
    overrideWorldGravity {z=-980}
    simSpaceSettings {maxAngularAcceleration=50, maxAngularVelocity=10, maxLinearAcceleration=50, maxLinearVelocity=10, worldAlpha=1}
   }
   n13 = RigidBody(ComponentPose=@n12, Alpha=Get shirt_l_alpha() -> (shirt_l_alpha:float)) {
    bClampLinearTranslationLimitToRefPose=true
    overridePhysicsAsset=&M_MED_CanineCronutMix_Physics_Shirt_l
    simulationSpace=BaseBoneSpace
    baseBoneRef {boneName=upperarm_l}
    overrideWorldGravity {z=-980}
    simSpaceSettings {maxAngularAcceleration=50, maxAngularVelocity=10, maxLinearAcceleration=50, maxLinearVelocity=10, worldAlpha=1}
   }
   Output Pose(Result=Component To Local(ComponentPose=@n13)) {layerGroup=None}
  ]}
 EventGraph {nodes [
   n1 = Event BlueprintUpdateAnimation()
   n2 = GetOwnerSkeletalMeshComponent()
   > Set Short_r_Alpha(Short_r_Alpha:double=Remap Float(
     Input=GetSocketTransform!(self=@n2, InSocketName=calf_r, TransformSpace=RTS_ParentBoneSpace) -> (
      ReturnValue_Location:Vector
      ReturnValue_Rotation_Roll:float
      ReturnValue_Rotation_Pitch:float
      ReturnValue_Rotation_Yaw:float
      ReturnValue_Scale:Vector
     ).ReturnValue_Rotation_Yaw
     OldRange=(X=45,Y=90)
     NewRange=(X=1,Y=0)
     Clamp=true
    ) -> (Output=~)) -> (Output_Get:double)
   > Set Short_l_Alpha(Short_l_Alpha:float=Remap Float(
     Input=GetSocketTransform(self=@n2, InSocketName=calf_l, TransformSpace=RTS_ParentBoneSpace).ReturnValue_Rotation_Yaw
     OldRange=(X=45,Y=90)
     NewRange=(X=1,Y=0)
     Clamp=true
    ) -> (Output=~)) -> (Output_Get:float)
   n3 = OR Boolean!(A=Get bIsSkydiveFloating(), B=Get bIsSkydiveDiveMode(), C:bool=Get bIsParachuteOpened()) -> (ReturnValue)
   n4 = MakeVector2D(X=SelectFloat(A=0.2, B=1, bPickA=@n3), Y=0)
   > Set shirt_r_alpha(shirt_r_alpha:double=Remap Float(
     Input=GetSocketTransform(self=@n2, InSocketName=upperarm_r, TransformSpace=RTS_ParentBoneSpace).ReturnValue_Rotation_Pitch
     OldRange=(X=0,Y=30)
     NewRange=@n4
     Clamp=true
    ) -> (Output=~)) -> (Output_Get:double)
   > Set shirt_l_alpha(shirt_l_alpha:float=Remap Float(
     Input=GetSocketTransform(self=@n2, InSocketName=upperarm_l, TransformSpace=RTS_ParentBoneSpace).ReturnValue_Rotation_Pitch
     OldRange=(X=0,Y=20)
     NewRange=@n4
     Clamp=true
    ) -> (Output=~)) -> (Output_Get:float)
   > n5 = Sequence!(execute) -> (then_0, then_1, then_2:Exec)
   > n6 = Branch(Condition=@n3)
   > Set SkyDiving_Alpha(SkyDiving_Alpha:double=1) -> (Output_Get:double)
   > Set Velocity Interp Speed(Velocity Interp Speed:double=40) -> (Output_Get:double)
   > Set WPO Wind(WPO Wind:double=1) -> (Output_Get:double)
   > Set Turbulence Noise(Turbulence Noise:float=400) -> (Output_Get:float)
   > Set R1_DragStrength(R1_DragStrength:float=1.25) -> (Output_Get:float)
   > Set R2_DragStrength(R2_DragStrength:float=3) -> (Output_Get:float)
   Set SkyDiving_Alpha(execute=@n6.else, SkyDiving_Alpha=0)
   > Set Velocity Interp Speed(Velocity Interp Speed=0)
   > Set WPO Wind(WPO Wind=0)
   > Set Turbulence Noise(Turbulence Noise=10000)
   > Set R1_DragStrength(R1_DragStrength=0)
   > Set R2_DragStrength(R2_DragStrength=0)
   n7 = Get Mesh!() -> (Mesh)
   Is Valid(exec=@n5.then_1, InputObject=@n7)
   n8 = Literal enum EAxis!(Enum:EAxis Enum=Y) -> (ReturnValue:EAxis Enum) {is=title K2Node_EnumLiteral}
   n9 = Get Overlay Mat Instance Dyanmic() -> (Overlay Mat Instance Dyanmic:Material Instance Dynamic Object Reference)
   n10 = Get Material Instance Dynamic() -> (Material Instance Dynamic:Material Instance Dynamic Object Reference)
   > Set WPO Globals Function(
    Material=@n10
    Overlay Material=@n9
    Skeletal Mesh=@n7
    "Local Noise Transform  Location"="0, 0, 0"
    Local Noise Offset="0, 0, 0"
    Local Turbulence Vector Transform Axis=@n8
   )
   > Set WPO Regional Function!(
    execute
    Material:Material Instance Dynamic Object Reference=@n10
    Overlay Material:Material Instance Dynamic Object Reference=@n9
    Skeletal Mesh:Skeletal Mesh Component Object Reference=@n7
    Velocity Transform:Name=pelvis
    Velocity Interp Speed:double=Get Velocity Interp Speed()
    Collision Transform:Name=pelvis
    Collision Transform Axis:EAxis Enum=@n8
    Invert Collision Transform Axis:bool=false
    Delta Time X:double=@n1.DeltaTimeX
    Previous Velocity Transform Location:"Vector (by ref)"="0, 0, 0"
    Previous Velocity Transform Velocity:"Vector (by ref)"="0, 0, 0"
    Region:int=1
   ) -> (then)
   > Set WPO Regional Function(
    Material=@n10
    Overlay Material=@n9
    Skeletal Mesh=@n7
    Velocity Transform=pelvis
    Velocity Interp Speed=Get Velocity Interp Speed()
    Collision Transform=pelvis
    Collision Transform Axis=@n8
    Invert Collision Transform Axis=false
    Delta Time X=@n1.DeltaTimeX
    Previous Velocity Transform Location="0, 0, 0"
    Previous Velocity Transform Velocity="0, 0, 0"
    Region=2
   )
   > SetScalarParameterValue(self=[@n10, @n9], ParameterName=ClothSim_WindStrength, Value=Get WPO Wind() -> (WPO Wind:double))
   > SetScalarParameterValue(self=[@n10, @n9], ParameterName=ClothSim_TurbulenceNoiseTileDivisor, Value=Get Turbulence Noise() -> (Turbulence Noise:float))
   > SetScalarParameterValue(self=[@n10, @n9], ParameterName=ClothSim_R1_DragStrength, Value=Get R1_DragStrength() -> (R1_DragStrength:float))
   > SetScalarParameterValue(self=[@n10, @n9], ParameterName=ClothSim_R2_DragStrength, Value=Get R2_DragStrength() -> (R2_DragStrength:float))
   n11 = Event BlueprintInitializeAnimation()
   n12 = GetPartSkeletalMeshComponent()
   n13 = Set Mesh(execute=[@n11.then, @n16.then], Mesh:Skeletal Mesh Component Object Reference=@n12) -> (Output_Get:Skeletal Mesh Component Object Reference)
   > n14 = Get Dynamic Material Instance(SkeletalMeshComponent=@n13)
   > Set Material Instance Dynamic(Material Instance Dynamic:Material Instance Dynamic Object Reference=@n14) -> (Output_Get:Material Instance Dynamic Object Reference)
   > n15 = GetOrCreateDynamicOverlayMaterial(MeshComponent=@n12, MaterialName=None)
   > Set Overlay Mat Instance Dyanmic(Overlay Mat Instance Dyanmic:Material Instance Dynamic Object Reference=@n15) -> (Output_Get:Material Instance Dynamic Object Reference)
   > Set Velocity Interp Speed(Velocity Interp Speed=0)
   > Set WPO Wind(WPO Wind=0)
   > Set Turbulence Noise(Turbulence Noise=0)
   n16 = Event OnCharacterCustomizationLoaded()
  ]}
]
variables [
 Material Instance Dynamic {type=MaterialInstanceDynamic}
 Mesh {type=SkeletalMeshComponent}
 Overlay Mat Instance Dyanmic {type=MaterialInstanceDynamic}
 R1_DragStrength {type=number}
 R2_DragStrength {type=number}
 shirt_l_alpha {type=number}
 shirt_r_alpha {type=number}
 Short_l_Alpha {type=number, default=1}
 Short_r_Alpha {type=number}
 SkyDiving_Alpha {type=number}
 Turbulence Noise {type=number}
 Velocity Interp Speed {type=number}
 WPO Wind {type=number}
]
